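# Cheap additive, unlit fog material for Niagara sprite particles, drawing a soft round glow tinted by particle color.
# A radial sphere mask times particle alpha is depth-faded, multiplied by particle RGB and passed to MF_FX_EmissiveAdjustments.
# A quality switch skips the depth fade on Low.
# A panning caustics-texture branch driven by dynamic parameters feeds nothing, so no smoke or chromatic split reaches that chain.
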Material /Game/Effects/Fort_Effects/Materials/Generic/M_Chromatic_Fog_Particles_Cheap
graph {
 properties {
  bUseTranslucencyVertexFog=false
  bUsedWithNiagaraSprites=true
  blendMode=BLEND_Additive
  pixelDepthOffsetMode=PDOM_Legacy
  shadingModel=MSM_Unlit
  usedShadingModels=Unlit
 }
 nodes [
  Constant() {r=5000}
  n1 = ScalarParameter() {defaultValue=0.45, parameterName=EdgeSoften}
  n2 = Multiply(
   A=SphereMask(A=TextureCoordinate(), B=Constant2Vector() {g=0.5, r=0.5}, Radius=@n1, Hardness=ScalarParameter() {defaultValue=0.1, parameterName=Hardness})
   B=ParticleColor().A
  )
  n3 = DepthFade(Opacity=@n2, FadeDistance=ScalarParameter() {defaultValue=2500, group=Opacity, parameterName=Opacity_Depth_FadeDistance})
  MaterialFunctionCall(Emissive=Multiply(A=QualitySwitch(Default=@n3, Low=@n2), B=ParticleColor().RGB)) {materialFunction=&MF_FX_EmissiveAdjustments}
  n4 = DynamicParameter() {paramNames [TileX,TileY,OffsetX,OffsetY]}
  n5 = TextureCoordinate()
  n6 = DynamicParameter() {paramNames [TileX,TileY,OffsetX,OffsetY]}
  n7 = AppendVector(A=@n6.TileX, B=@n6.TileX)
  n8 = DynamicParameter() {paramNames [TileX,TileY,OffsetX,OffsetY]}
  n9 = AppendVector(A=@n8.OffsetX, B=@n8.OffsetX)
  n10 = DynamicParameter() {parameterIndex=2, defaultValue=<rgba 0.05 0.05 1 1>, paramNames [PanSpeedX, PanSpeedY, ChromaAmount, RedGreenAmount]}
  n11 = Add(
   A=AppendVector(A=@n4.OffsetX, B=@n4.OffsetX)
   B=Panner(Coordinate=Multiply(A=@n5, B=@n7), Speed=Multiply(A=@n9, B=@n10.PanSpeedX)) {speedX=0.02, speedY=0.15}
  )
  n12 = DynamicParameter() {parameterIndex=1, defaultValue=<rgba 0.05 0.05 1 1>, paramNames [Contrast, DistortAmount, OffsetX, OffsetY]}
  Multiply(
   A=@n1
   B=Power(
    Base=TextureSampleParameter2D(UVs=@n11) {parameterName=SmokeTexture, texture=&T_ShieldCaustics}.G
    Exp=@n12.Contrast
   )
  )
 ]
}
