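# Generic masked, lit material for metal chunk effects on Niagara mesh particles.
# Particle color RGB tints a diffuse texture, and a roughness texture is scaled by a dynamic parameter.
# The Erode dynamic parameter drives the Erode_Alpha function on the opacity texture's green channel, multiplied by particle alpha for per-particle dissolve.
# Metallic and Specular are scalar parameters, and all three texture parameters default to a tiling noise.
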Material /Game/Effects/Fort_Effects/Materials/Generic/M_FX_Generic_Chunk_Metal
graph {
 properties {bUsedWithNiagaraMeshParticles=true, blendMode=BLEND_Masked, pixelDepthOffsetMode=PDOM_Legacy, usedShadingModels=Default Lit}
 nodes [
  n1 = ParticleColor()
  Multiply(A=@n1.RGB, B=TextureSampleParameter2D() {parameterName=DiffuseTexture, texture=/Game/Global/Textures/Default/Noise/T_Tiling_Noise_01}.RGB)
  Multiply(
   A=TextureSampleParameter2D() {parameterName=RoughnessTexture, texture=&T_Tiling_Noise_01}.RGB
   B=DynamicParameter() {paramNames [Erode,Roughness,Param3,Param4]}.Roughness
  )
  Multiply(
   A=@n1.A
   B=MaterialFunctionCall!(
    Dynamic Param=DynamicParameter() {paramNames [Erode,Roughness,Param3,Param4]}.Erode
    Alpha=TextureSampleParameter2D() {parameterName=OpacityTexture, texture=&T_Tiling_Noise_01}.G
   ) {materialFunction=/Game/Effects/Fort_Effects/Materials/_Functions/Erode_Alpha}.Result
  )
  ScalarParameter() {defaultValue=1, parameterName=Metallic}
  ScalarParameter() {defaultValue=1, parameterName=Specular}
 ]
}
