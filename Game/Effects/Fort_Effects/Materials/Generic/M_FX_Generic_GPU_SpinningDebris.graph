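# Masked, two-sided lit material for GPU sprite debris particles that appear to tumble as they move.
# A random sub-UV cell from a dust mote atlas supplies alpha, with particle color and scalar surface parameters.
# Each sprite rotates about its center on one of three randomly picked axes, with angle driven by distance from world origin and particle random rather than time.
# Shading path switches replace the randomness and rotation with constants on mobile.
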Material /Game/Effects/Fort_Effects/Materials/Generic/M_FX_Generic_GPU_SpinningDebris
graph {
 properties {
  bEnableResponsiveAA=true
  bUsedWithParticleSprites=true
  blendMode=BLEND_Masked
  opacityMaskClipValue=0.1
  pixelDepthOffsetMode=PDOM_Legacy
  translucencyLightingMode=TLM_VolumetricPerVertexDirectional
  twoSided=true
  usedShadingModels=Default Lit
 }
 nodes [
  ParticleColor()
  n1 = ScalarParameter() {defaultValue=2, parameterName="Rows (S)"}
  n2 = ScalarParameter() {defaultValue=4, parameterName="Columns (S)"}
  n3 = ShadingPathSwitch(Default=ParticleRandom(), Mobile=Constant()) {desc=Migrated from FeatureLevelSwitch}
  Power(Base=TextureSample(UVs=MaterialFunctionCall!(UseTexCoord 1, Rows=@n1, Columns=@n2, Particle Random Value=@n3) {materialFunction=/Game/Effects/Fort_Effects/Materials/_Functions/MF_RandomSubUV_GPU_Sprites}.Result) {texture=/Game/Effects/Fort_Effects/Textures/T_DustMotes_4x2_01}.A) {constExponent=0.4}
  ScalarParameter() {parameterName=Metallic}
  ScalarParameter() {defaultValue=0.25, parameterName=Specular}
  ScalarParameter() {defaultValue=0.6, parameterName=Roughness}
  n4 = MaterialFunctionCall!(
   Random Picker=ParticleRandom()
   Texture Input
   Option 1=VectorParameter() {parameterName=NormalizeRotationAxis_1, defaultValue=<rgba 1 0.3 -0.2 0>}.RGB
   Option 2=VectorParameter() {parameterName=NormalizeRotationAxis_2, defaultValue=<rgba -0.44 1 0.13 0>}.RGB
   Option 3=VectorParameter() {parameterName=NormalizeRotationAxis_3, defaultValue=<rgba 0.06 -0.22 1 0>}.RGB
  ) {materialFunction=/Game/Effects/Fort_Effects/Materials/_Functions/MF_RGB_Random_01}
  n5 = Divide(A=Distance(A=Constant3Vector(), B=ParticlePositionWS()), B=ScalarParameter() {defaultValue=1000, parameterName=Spin Scale})
  n6 = Clamp(None=ParticleRandom()) {minDefault=0.55}
  ShadingPathSwitch(
   Default=RotateAboutAxis(
    NormalizedRotationAxis=Normalize(VectorInput=@n4.Vec3 Out Only)
    RotationAngle=Multiply(A=@n5, B=@n6)
    PivotPoint=ParticlePositionWS()
    Position=WorldPosition().XYZ
   )
   Mobile=Constant()
  ) {desc=Migrated from FeatureLevelSwitch}
 ]
}
