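# Unlit translucent material for Niagara mesh particles, named for spiky spherical ground-impact effects, with particle color feeding an emissive adjustment function.
# Two Fresnel bands fade center and rim, one shaped by dynamic parameter MiddleFade.
# They mask two multiplied, vertically panning smoke samples, driven by PanSpeed and Tiling, and particle alpha scales the resulting opacity.
# A shading path switch applies depth fade by default but skips it on mobile.
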Material /Game/Effects/Fort_Effects/Materials/Generic/M_FX_Generic_GroundImpact_Sphere_Spikey
graph {
 properties {
  bUsedWithNiagaraMeshParticles=true
  blendMode=BLEND_Translucent
  pixelDepthOffsetMode=PDOM_Legacy
  shadingModel=MSM_Unlit
  usedShadingModels=Unlit
 }
 nodes [
  MaterialFunctionCall(Emissive=ParticleColor().RGB) {materialFunction=&MF_FX_EmissiveAdjustments}
  n1 = Fresnel(ExponentIn=Constant() {r=0.8}) {baseReflectFraction=0.001}
  n2 = Multiply(A=OneMinus(None=@n1), B=@n1)
  n3 = ParticleColor()
  n4 = Fresnel(ExponentIn=DynamicParameter() {defaultValue=<rgba 3 1 1 1>, paramNames [MiddleFade,EdgeFade,PanSpeed,Tiling]}.MiddleFade) {baseReflectFraction=0.001}
  n5 = OneMinus(None=@n4)
  n6 = Power(Base=Multiply(A=Multiply(A=@n5, B=@n4), B=Constant() {r=1}))
  n7 = AppendVector(A=Constant() {r=1}, B=DynamicParameter() {defaultValue=<rgba 3 1 1 1>, paramNames [MiddleFade,EdgeFade,PanSpeed,Tiling]}.Tiling)
  n8 = Multiply(A=Time(), B=DynamicParameter() {defaultValue=<rgba 3 1 1 1>, paramNames [MiddleFade,EdgeFade,PanSpeed,Tiling]}.PanSpeed)
  n9 = Panner(Coordinate=Multiply(A=TextureCoordinate(), B=@n7), Time=@n8) {speedY=-1}
  n10 = Panner(Coordinate=Multiply(A=TextureCoordinate() {uTiling=2}, B=@n7), Time=@n8) {speedY=-1}
  n11 = Multiply(
   A=Multiply(
    A=TextureSample(UVs=@n9) {texture=&T_Inky_Smoke_Tile}.G
    B=TextureSample(UVs=@n10) {texture=&T_Inky_Smoke_Tile}.G
   )
   B=Constant() {r=4}
  )
  n12 = Multiply(A=Power(Base=Multiply(A=@n2, B=@n3.A) {constA=1}), B=Clamp(None=Multiply(A=@n6, B=@n11)))
  ShadingPathSwitch(Default=DepthFade(Opacity=@n12), Mobile=@n12) {desc=Migrated from FeatureLevelSwitch}
 ]
}
