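# Translucent, two-sided lit material for Niagara sprite particles that draws leaves from a SubUV flipbook texture.
# The texture's green channel is tinted by particle color and hue-shifted through a dynamic parameter, with optional fake shadows that shade each face differently via TwoSidedSign.
# A static switch picks that result or a hue-shifted, desaturatable particle-color glow for the emissive adjustments function.
# Texture and particle alpha form opacity with optional distance fade; a normal map is also optional.
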Material /Game/Effects/Fort_Effects/Materials/Generic/M_FX_Generic_Leaf_SubUV_Niagara
graph {
 properties {
  bCastRayTracedShadows=false
  bUsedWithNiagaraSprites=true
  blendMode=BLEND_Translucent
  pixelDepthOffsetMode=PDOM_Legacy
  translucencyLightingMode=TLM_VolumetricPerVertexDirectional
  twoSided=true
  usedShadingModels=Default Lit
 }
 nodes [
  Constant()
  DynamicParameter() {paramNames [EmmisveAmount, HueShiftAmount, SaturationAmount, Param4]}
  n1 = TextureSampleParameterSubUV!(UVs, Apply View MipBias) {parameterName=LeavesTexture, texture=/Game/Effects/Fort_Effects/Textures/Leaf/T_Leaf_SubUV}
  n2 = ParticleColor()
  n3 = MaterialFunctionCall(
   Hue Shift Percentage=DynamicParameter() {paramNames [EmmisveAmount, HueShiftAmount, SaturationAmount, Param4]}.HueShiftAmount
   Texture=MaterialFunctionCall(DiffuseColor=Multiply(A=@n1.G, B=@n2.RGB)) {materialFunction=&ConvertFromDiffSpec}.BaseColor
  ) {materialFunction=&HueShift}
  n4 = Multiply(
   A=LinearInterpolate(
    A=ScalarParameter() {defaultValue=0.75, parameterName=Shadow_Low}
    B=ScalarParameter() {defaultValue=1, parameterName=Shadow_High}
    Alpha=TwoSidedSign()
   ) {constA=0.75}
   B=@n3.Result
  )
  n5 = MaterialFunctionCall(
   Hue Shift Percentage=DynamicParameter() {paramNames [EmmisveAmount, HueShiftAmount, SaturationAmount, Param4]}.HueShiftAmount
   Texture=Multiply(A=ParticleColor().RGB, B=DynamicParameter() {paramNames [EmmisveAmount, HueShiftAmount, SaturationAmount, Param4]}.EmmisveAmount)
  ) {materialFunction=&HueShift}
  n6 = DynamicParameter() {paramNames [EmmisveAmount, HueShiftAmount, SaturationAmount, Param4]}
  MaterialFunctionCall(Emissive=StaticSwitchParameter(
    True=StaticSwitchParameter(True=@n4, False=@n3.Result) {parameterName=UseFakeShadows?}
    False=Desaturation(None=@n5.Result, Fraction=@n6.SaturationAmount)
   ) {parameterName=Emissive Switch}) {materialFunction=&MF_FX_EmissiveAdjustments}
  ParticleColor()
  ScalarParameter() {parameterName=Spec}
  ScalarParameter() {defaultValue=1, parameterName=Roughness}
  n7 = ScalarParameter() {defaultValue=500, parameterName="Fade_Start_Distance (S)"}
  n8 = ScalarParameter() {defaultValue=100, parameterName="Fade_End_Distance (S)"}
  n9 = Multiply(A=@n1.A, B=ParticleColor().A)
  StaticSwitchParameter(
   True=Multiply(
    A=MaterialFunctionCall(Fade_Start_Distance=@n7, Fade_End_Distance=@n8) {materialFunction=&MF_Distance_SyncedFade_FX}.Result Inverted
    B=@n9
   )
   False=@n9
  ) {parameterName=UseDistanceFade?}
  StaticSwitchParameter(
   True=TextureSampleParameterSubUV() {parameterName=NormalTexture, texture=/Game/Effects/Fort_Effects/Textures/Leaf/T_Cherry_Petals_64_N_01}.RGB
   False=Constant()
  ) {parameterName=UseNormalMap?}
 ]
}
