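# Unlit translucent particle material for cheap emissive smoke, usable on Cascade and Niagara sprites.
# Particle color is tinted by morning, day, evening and night parameters, blended by Color_Time_of_Day, then passed through MF_FX_EmissiveAdjustments.
# Alpha gets the same time-of-day blend, multiplied by particle alpha and an 8x8 SubUV smoke flipbook's alpha.
# The UseNearFade static switch optionally multiplies in a near-camera fade.
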Material /Game/Effects/Fort_Effects/Materials/Generic/M_FX_Generic_SwirlSmoke_Emissive_Cheap_01
graph {
 properties {
  bUsedWithNiagaraSprites=true
  bUsedWithParticleSprites=true
  blendMode=BLEND_Translucent
  pixelDepthOffsetMode=PDOM_Legacy
  shadingModel=MSM_Unlit
  translucencyLightingMode=TLM_VolumetricPerVertexDirectional
  usedShadingModels=Unlit
 }
 nodes [
  n1 = VectorParameter() {parameterName=Color_Morning, defaultValue=<rgba 1 0.5451 0.4 0>}
  n2 = ParticleColor()
  n3 = VectorParameter() {parameterName=Color_Day, defaultValue=<rgba 1 1 1 1>}
  n4 = VectorParameter() {parameterName=Color_Evening, defaultValue=<rgba 1 0.4263 0.4 0>}
  n5 = VectorParameter() {parameterName=Color_Night, defaultValue=<rgba 0.4 0.5658 1 0>}
  MaterialFunctionCall(Emissive=MaterialFunctionCall(
    Color_Morning=Multiply(A=@n1.RGB, B=@n2.RGB)
    Color_Day=Multiply(A=@n2.RGB, B=@n3.RGB)
    Color_Evening=Multiply(A=@n2.RGB, B=@n4.RGB)
    Color_Night=Multiply(A=@n2.RGB, B=@n5.RGB)
   ) {materialFunction=&Color_Time_of_Day}.Result) {materialFunction=&MF_FX_EmissiveAdjustments}
  n6 = Multiply(A=ScalarParameter() {defaultValue=1.5, parameterName=Alpha_Morning}, B=@n2.A)
  n7 = Multiply(A=ScalarParameter() {defaultValue=1, parameterName=Alpha_Day}, B=@n2.A)
  n8 = Multiply(A=ScalarParameter() {defaultValue=2, parameterName=Alpha_Evening}, B=@n2.A)
  n9 = Multiply(A=@n2.A, B=ScalarParameter() {defaultValue=2, parameterName=Alpha_Night})
  n10 = ScalarParameter() {defaultValue=250, parameterName="Distance (S)"}
  Multiply(
   A=Multiply(
    A=MaterialFunctionCall(Color_Morning=@n6, Color_Day=@n7, Color_Evening=@n8, Color_Night=@n9) {materialFunction=&Color_Time_of_Day}.Result
    B=ParticleSubUV!(UVs, Tex, Apply View MipBias) {texture=/Game/Effects/Fort_Effects/BluePrints/BeeHive/T_SmokeFlip_03_8x8}.A
   )
   B=StaticSwitchParameter(
    True=MaterialFunctionCall(Distance=@n10) {materialFunction=&Near_Camera_Fade}.Result
    False=Constant() {r=1}
   ) {parameterName=UseNearFade}
  )
 ]
}
