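# Cheap unlit translucent smoke particle material for sprite and Niagara effects, with per-particle random hue and optional time-of-day tinting.
# An 8x8 SubUV smoke texture drives brightness and opacity, combined with particle color, optional world-space sphere glow and emissive adjustments.
# Static switches add depth, near-camera and distance fades, plus SmoothLod opacity and WPO scaling.
# Shading-path and quality switches drop time-of-day color or fades on cheaper paths.
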Material /Game/Effects/Fort_Effects/Materials/Generic/M_FX_Generic_SwirlSmoke_Emissive_RandomColor_Cheap
graph {
 properties {
  allowTranslucentCustomDepthWrites=true
  bUsedWithNiagaraSprites=true
  bUsedWithParticleSprites=true
  blendMode=BLEND_Translucent
  opacityMaskClipValue=0.01
  pixelDepthOffsetMode=PDOM_Legacy
  shadingModel=MSM_Unlit
  translucencyLightingMode=TLM_VolumetricPerVertexDirectional
  usedShadingModels=Unlit
 }
 nodes [
  n1 = ActorPositionWS()
  n2 = Transform(None=VectorParameter() {group=Glow, parameterName=Glow Offset, defaultValue=<rgba 0 0 0 1>}.RGB) {transformSourceType=TRANSFORMSOURCE_Local}
  n3 = Power(
   Base=SphereMask(
    A=WorldPosition().XYZ
    B=Add(A=@n1, B=@n2)
    Radius=ScalarParameter() {defaultValue=25, group=Glow, parameterName=Glow Radius}
    Hardness=ScalarParameter() {defaultValue=25, group=Glow, parameterName=Glow Hardness}
   )
   Exp=ScalarParameter() {defaultValue=25, group=Glow, parameterName=Glow Mask Power}
  )
  n4 = ParticleRandom()
  n5 = Constant() {r=1}
  n6 = Multiply(
   A=ScalarParameter() {defaultValue=1, parameterName=HueShiftAmount}
   B=StaticSwitchParameter(
    True=ShadingPathSwitch(Default=@n4, Mobile=@n5) {desc=Migrated from FeatureLevelSwitch}
    False=DynamicParameter() {paramNames [RandomColor,Param2,Param3,Param4]}.RandomColor
   ) {parameterName=UseforGpuParticles?}
  )
  n7 = VectorParameter() {parameterName=Color_Morning, defaultValue=<rgba 1 0.5451 0.4 0>}
  n8 = ParticleColor()
  n9 = VectorParameter() {parameterName=Color_Day, defaultValue=<rgba 1 1 1 1>}
  n10 = VectorParameter() {parameterName=Color_Evening, defaultValue=<rgba 1 0.4263 0.4 0>}
  n11 = VectorParameter() {parameterName=Color_Night, defaultValue=<rgba 0.4 0.5658 1 0>}
  n12 = StaticSwitchParameter(
   True=MaterialFunctionCall(
    Color_Morning=Multiply(A=@n7.RGB, B=@n8.RGB)
    Color_Day=Multiply(A=@n8.RGB, B=@n9.RGB)
    Color_Evening=Multiply(A=@n8.RGB, B=@n10.RGB)
    Color_Night=Multiply(A=@n8.RGB, B=@n11.RGB)
   ) {materialFunction=&Color_Time_of_Day}.Result
   False=@n8.RGB
  ) {defaultValue=true, parameterName=UseTOD?}
  n13 = ShadingPathSwitch(Default=@n12, Deferred=@n12, Forward=@n8.RGB, Mobile=@n8.RGB)
  n14 = ParticleSubUV!(UVs, Tex, Apply View MipBias) {samplerType=SAMPLERTYPE_LinearColor, texture=/Game/Effects/Fort_Effects/Textures/Smoke/T_Smoke_8x8_01}
  n15 = Multiply(
   A=StaticSwitchParameter(
    True=MaterialFunctionCall(Hue Shift Percentage=@n6, Texture=@n13) {materialFunction=&HueShift}.Result
    False=@n13
   ) {defaultValue=true, parameterName=UseHueShift?}
   B=LinearInterpolate(A=ScalarParameter() {parameterName=MinBrightness}, Alpha=@n14.G)
  )
  n16 = StaticSwitchParameter(
   True=Multiply(A=VectorParameter() {group=Glow, parameterName=GlowColor, defaultValue=<rgba 1 0.525 0.2916 1>}.RGB, B=Add(A=@n3, B=@n15))
   False=@n15
  ) {parameterName=AddWorldSpaceGlow?}
  LinearInterpolate(
   A=@n16
   B=MaterialFunctionCall(Emissive=@n16) {materialFunction=&MF_FX_EmissiveAdjustments}.Result
   Alpha=ScalarParameter() {defaultValue=1, parameterName=UseEmissiveAdjustments}
  )
  n17 = LinearInterpolate(Alpha=MaterialFunctionCall!(
    Blend Range=ScalarParameter() {defaultValue=200, parameterName=CloseFade}
    Start Offset=ScalarParameter() {defaultValue=-150, parameterName=CloseFade_Offset}
   ) {materialFunction=/Engine/Functions/Engine_MaterialFunctions03/Particles/Distance_Blend}.Result)
  LinearInterpolate(
   A=@n17
   Alpha=MaterialFunctionCall(Blend Range=Constant() {r=5000}, Start Offset=Constant() {r=-1050}) {materialFunction=&Distance_Blend}.Result
  ) {constB=0}
  n18 = MaterialFunctionCall!(
   NearFadeLength=ScalarParameter() {defaultValue=1500, group=ScaleSize, parameterName="NearFadeLength (S)", sortPriority=1}
   NearFadeOffset=ScalarParameter() {group=ScaleSize, parameterName=NearFadeOffset, sortPriority=1}
   NearFadeCurve=ScalarParameter() {group=ScaleSize, parameterName=NearFadeCurve, sortPriority=1}
   NearFadeStartingScale=ScalarParameter() {defaultValue=1, group=ScaleSize, parameterName="NearFadeStartingScale (S)"}
   NearFadeEndingScale
   FarFadeLength=ScalarParameter() {defaultValue=2500, group=ScaleSize, parameterName="FarFadeLength (S)", sortPriority=4}
   FarFadeOffset
   FarFadeCurve
   FarFadeEndingScale=ScalarParameter() {defaultValue=4, group=ScaleSize, parameterName=FarFadeEndingScale_HIGH, sortPriority=3}
   Opacity Near Distance Bias=ScalarParameter() {group=ScaleSize, parameterName=OpacityNear_Bias, sortPriority=5}
   Opacity Far Distance Bias=ScalarParameter() {group=ScaleSize, parameterName=OpacityFar_Bias, sortPriority=6}
  ) {materialFunction=/Game/Effects/Fort_Effects/Materials/_Functions/MF_SmoothLod}
  n19 = Constant() {r=1}
  n20 = Multiply(A=ScalarParameter() {defaultValue=1.5, parameterName=Alpha_Morning}, B=@n8.A)
  n21 = Multiply(A=ScalarParameter() {defaultValue=1, parameterName=Alpha_Day}, B=@n8.A)
  n22 = Multiply(A=ScalarParameter() {defaultValue=2, parameterName=Alpha_Evening}, B=@n8.A)
  n23 = Multiply(A=@n8.A, B=ScalarParameter() {defaultValue=2, parameterName=Alpha_Night})
  n24 = Multiply(
   A=StaticSwitchParameter(
    True=MaterialFunctionCall(Color_Morning=@n20, Color_Day=@n21, Color_Evening=@n22, Color_Night=@n23) {materialFunction=&Color_Time_of_Day}.Result
    False=@n8.A
   ) {defaultValue=true, parameterName=UseTOD?}
   B=@n14.R
  )
  n25 = ScalarParameter() {defaultValue=250, parameterName="Distance (S)"}
  n26 = Multiply(
   A=@n24
   B=StaticSwitchParameter(
    True=MaterialFunctionCall(Distance=@n25) {materialFunction=&Near_Camera_Fade}.Result
    False=Constant() {r=1}
   ) {parameterName=UseNearFade}
  )
  n27 = ScalarParameter() {defaultValue=250, parameterName=FadeDistance}
  n28 = ShadingPathSwitch(
   Default=StaticSwitchParameter(True=DepthFade(Opacity=@n26, FadeDistance=@n27), False=@n26) {defaultValue=true, parameterName=UseDepthFade}
   Mobile=@n26
  ) {desc=Migrated from FeatureLevelSwitch}
  n29 = QualitySwitch(Default=@n28, Low=@n26, High=@n28, Medium=@n28)
  Multiply(
   A=StaticSwitchParameter(True=@n17, False=StaticSwitchParameter(True=@n18.Opacity, False=@n19) {parameterName=Scale_Opacity_When_Close?}) {parameterName=UseLowendCloseUpFade?}
   B=ShadingPathSwitch(Default=@n29, Deferred=@n29, Forward=@n24)
  )
  StaticSwitchParameter(True=@n18.WPO, False=@n19) {parameterName=Scale_Down_When_Close?}
 ]
}
