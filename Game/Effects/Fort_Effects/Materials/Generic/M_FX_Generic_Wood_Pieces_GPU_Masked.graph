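# Masked, two-sided, lit sprite-particle material for tumbling wood debris pieces.
# Each piece rotates around one of three randomly picked axes, with angle driven by particle distance from world origin and a random factor, so movement creates spin.
# Two axis-wise sphere masks and noise shape the cutout, multiplied by particle alpha and a distance fade.
# Particle color tints tiled barn wood; the mobile path drops the wood texture and fade.
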Material /Game/Effects/Fort_Effects/Materials/Generic/M_FX_Generic_Wood_Pieces_GPU_Masked
graph {
 properties {
  bEnableResponsiveAA=true
  bUsedWithParticleSprites=true
  blendMode=BLEND_Masked
  opacityMaskClipValue=0.1
  pixelDepthOffsetMode=PDOM_Legacy
  translucencyLightingMode=TLM_VolumetricPerVertexDirectional
  twoSided=true
  usedShadingModels=Default Lit
 }
 nodes [
  n1 = ParticleRandom()
  n2 = VectorParameter() {parameterName=NormalizeRotationAxis_1, defaultValue=<rgba 1 0.3 -0.2 0>}
  n3 = VectorParameter() {parameterName=NormalizeRotationAxis_2, defaultValue=<rgba -0.44 1 0.13 0>}
  n4 = VectorParameter() {parameterName=NormalizeRotationAxis_3, defaultValue=<rgba 0.06 -0.22 1 0>}
  n5 = Distance(A=Constant3Vector(), B=ParticlePositionWS())
  n6 = ScalarParameter() {defaultValue=2500, parameterName=Spin Scale}
  n7 = ParticleRandom()
  RotateAboutAxis(
   NormalizedRotationAxis=Normalize(VectorInput=MaterialFunctionCall!(Random Picker=@n1, Texture Input, Option 1=@n2.RGB, Option 2=@n3.RGB, Option 3=@n4.RGB) {materialFunction=/Game/Effects/Fort_Effects/Materials/_Functions/MF_RGB_Random_01}.Vec3 Out Only)
   RotationAngle=Multiply(A=Divide(A=@n5, B=@n6), B=Clamp(None=@n7) {minDefault=0.55})
   PivotPoint=ParticlePositionWS()
   Position=WorldPosition().XYZ
  )
  ScalarParameter() {parameterName=Metallic}
  ScalarParameter() {defaultValue=0.25, parameterName=Specular}
  ScalarParameter() {defaultValue=0.6, parameterName=Roughness}
  n8 = Power(Base=TextureSample(UVs=TextureCoordinate() {vTiling=0.1}) {texture=/Game/Global/Textures/Default/Noise/T_FireNoise_01}.G) {constExponent=1}
  n9 = Multiply(
   A=SphereMask(
    A=TextureCoordinate() {vTiling=0}
    B=Constant2Vector() {r=0.5}
    Radius=ScalarParameter() {defaultValue=0.45, parameterName=WoodMask_Radius}
    Hardness=ScalarParameter() {defaultValue=0.5, parameterName=WoodMask_Hardness}
   )
   B=SphereMask(
    A=TextureCoordinate() {uTiling=0}
    B=Constant2Vector() {g=0.5}
    Radius=ScalarParameter() {defaultValue=0.45, parameterName=WoodMask_Radius_1}
    Hardness=ScalarParameter() {parameterName=WoodMask_Hardness_1}
   )
  )
  n10 = Multiply(A=LinearInterpolate(A=Multiply(A=@n8, B=@n9), B=@n9, Alpha=@n9), B=ParticleColor().A)
  ShadingPathSwitch(
   Default=MaterialFunctionCall(
    Opacity In=@n10
    Position=ParticlePositionWS()
    Fade_Start_Distance=ScalarParameter() {defaultValue=150, parameterName=Fade Start Scale}
    Fade_End_Distance=ScalarParameter() {defaultValue=120, parameterName=Fade End Scale}
   ) {materialFunction=&MF_Distance_SyncedFade_FX}.Result
   Mobile=@n10
  ) {desc=Migrated from FeatureLevelSwitch}
  n11 = ParticleColor()
  n12 = TextureCoordinate() {uTiling=3, vTiling=3}
  ShadingPathSwitch(Default=Multiply(A=@n11.RGB, B=TextureSample(UVs=@n12) {texture=/Game/Environments/Sets/Barn/Textures/T_Barn_Wood_1_D}.RGB), Mobile=@n11.RGB) {desc=Migrated from FeatureLevelSwitch}
  TextureCoordinate()
 ]
}
