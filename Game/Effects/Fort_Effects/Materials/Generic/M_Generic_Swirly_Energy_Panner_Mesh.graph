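# Translucent, unlit, two-sided Niagara mesh-particle material that renders panning, swirling energy tinted by particle color and an optional curve-atlas gradient.
# Layered noise distorts flame, turbulence and mask textures, with offset flame samples splitting RGB for color fringing.
# Opacity combines V-gradient and view-angle fades, plus depth fade outside forward and mobile paths.
# Sine and noise displace vertices along normals, and a switch lets Dynamic Parameters drive panning time.
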Material /Game/Effects/Fort_Effects/Materials/Generic/M_Generic_Swirly_Energy_Panner_Mesh
graph {
 properties {
  bCastRayTracedShadows=false
  bUseTranslucencyVertexFog=false
  bUsedWithNiagaraMeshParticles=true
  blendMode=BLEND_Translucent
  pixelDepthOffsetMode=PDOM_Legacy
  shadingModel=MSM_Unlit
  twoSided=true
  usedShadingModels=Unlit
 }
 nodes [
  n1 = ParticleColor()
  n2 = Multiply(
   A=Time()
   B=StaticSwitchParameter(True=DynamicParameter() {defaultValue=<rgba 150 1 1 1>, paramNames [DepthFade,Time,Param3,Param4]}.Time, False=Constant() {r=0.5}) {parameterName=UseDPAsTime?}
  ) {constB=0.5}
  n3 = TextureSample(UVs=Panner(Coordinate=TextureCoordinate() {uTiling=2, vTiling=0.5}) {speedX=0.15, speedY=-0.3}) {samplerType=SAMPLERTYPE_LinearGrayscale, texture=/Engine/Functions/Engine_MaterialFunctions02/ExampleContent/Textures/LowResBlurredNoise}
  n4 = Multiply(A=@n3.R) {constB=0.1}
  n5 = TextureCoordinate() {vTiling=0.5}
  n6 = TextureSample(UVs=Add(A=@n4, B=Panner(Coordinate=@n5) {speedX=-0.1, speedY=0.4})) {texture=&T_Inky_Smoke_Tile}
  n7 = Multiply(A=Multiply(A=@n4, B=Multiply(A=@n6.G) {constB=0.2}) {constB=2}) {constB=10}
  n8 = Add(A=Panner(Coordinate=TextureCoordinate() {uTiling=0.5, vTiling=0.3}, Time=@n2) {speedY=0.5}, B=@n7)
  n9 = TextureSample(UVs=Subtract(A=@n8) {constB=0.005}) {samplerType=SAMPLERTYPE_LinearGrayscale, texture=/Game/Weapons/FORT_Melee/Pickaxe_Cube_Ninja_Male/FX/T_FluidFlameTile}
  n10 = Panner(Coordinate=TextureCoordinate() {uTiling=0.7, vTiling=0.4}, Time=@n2) {speedY=2}
  n11 = Power(Base=TextureSample(UVs=Add(A=@n10, B=@n7)) {texture=/Game/Effects/Fort_Effects/Textures/Plasma/T_Turbulence_Stretched}.G)
  n12 = Multiply(A=@n7, B=Constant2Vector() {g=0.5})
  n13 = MaterialFunctionCall(Float2=TextureCoordinate() {uTiling=0.7, vTiling=0.9}) {materialFunction=&BreakOutFloat2Components}
  n14 = MaterialFunctionCall(X=Add(A=@n13.R) {constB=0.15}, Y=Add(A=@n13.G) {constB=0.07}) {materialFunction=&MakeFloat2}
  n15 = Power(Base=TextureSample(UVs=Add(A=@n12, B=@n14.Result)) {samplerSource=SSM_Clamp_WorldGroupSettings, texture=/Game/WaterAndWind/Textures/WeaponFireVectorMask}.R)
  n16 = Multiply(A=Multiply(A=Add(A=@n9.G, B=@n11)) {constB=5}, B=Multiply(A=@n15) {constB=5})
  n17 = MaterialFunctionCall() {materialFunction=&LinearGradient}
  n18 = Multiply(A=Saturate(None=Multiply(A=@n16) {constB=10}), B=Power(Base=@n17.VGradient) {constExponent=3})
  n19 = Abs(None=DotProduct(A=CameraVectorWS(), B=PixelNormalWS()))
  n20 = DepthFade(
   Opacity=Multiply(A=@n18, B=Saturate(None=@n19))
   FadeDistance=DynamicParameter() {defaultValue=<rgba 150 1 1 1>, paramNames [DepthFade,Time,Param3,Param4]}.DepthFade
  ) {fadeDistanceDefault=50}
  Divide(A=Multiply(A=@n1.A, B=ShadingPathSwitch(Default=@n20, Deferred=@n20, Forward=@n18, Mobile=@n18))) {constB=3}
  n21 = CurveAtlasRowParameter(CurveTime=ComponentMask(None=TextureCoordinate()) {g=true}) {
   atlas=/Game/Athena/LinearColorCurves/VFX_CurveAtlas
   curve=/Game/Weapons/FORT_Melee/Pickaxe_Cube_Queen_Female/FX/Cube_Queen_EmissiveCurve
   defaultValue=41
  }
  n22 = StaticSwitchParameter(True=@n21, False=Constant() {r=1}) {defaultValue=true, parameterName=UseCurveforColor?}
  n23 = MaterialFunctionCall(
   X=TextureSample(UVs=Add(A=@n8) {constB=0.005}) {samplerType=SAMPLERTYPE_LinearGrayscale, texture=&T_FluidFlameTile}.G
   Y=TextureSample(UVs=@n8) {samplerType=SAMPLERTYPE_LinearGrayscale, texture=&T_FluidFlameTile}.G
   Z=@n9.G
  ) {materialFunction=&MakeFloat3}
  n24 = Power(Base=@n23.Result) {constExponent=7}
  n25 = Add(A=Add(A=Multiply(A=@n24) {constB=50}, B=Multiply(A=@n23.Result) {constB=2}), B=Multiply(A=@n11) {constB=10})
  n26 = Multiply(A=@n15, B=@n15) {constB=5}
  n27 = Add(A=LinearInterpolate(A=@n25, B=Multiply(A=@n25) {constB=5}, Alpha=Saturate(None=@n26))) {constB=0.1}
  MaterialFunctionCall(Emissive=Multiply(A=Multiply(A=@n22, B=@n27), B=@n1.RGB)) {materialFunction=&MF_FX_EmissiveAdjustments}
  n28 = Constant3Vector()
  n29 = Time()
  n30 = Sine(None=Add(A=Multiply(A=@n29) {constB=2}, B=Multiply(A=@n17.VGradient) {constB=2}) {constB=2})
  n31 = TextureSample(UVs=Panner(Time=@n2) {speedY=1.5}) {samplerType=SAMPLERTYPE_LinearGrayscale, texture=&LowResBlurredNoise}
  n32 = Add(
   A=MaterialFunctionCall(X=@n30, Y=@n30, Z=@n30) {materialFunction=&MakeFloat3}.Result
   B=@n31.RGB
  )
  n33 = OneMinus(None=@n17.VGradient)
  Multiply(A=VertexNormalWS(), B=Multiply(A=LinearInterpolate(A=@n28, B=@n32, Alpha=@n33)) {constB=10})
 ]
}
